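# Sound cue for the end of a quest device call, played as a non-critical in-game UI sound with narrative quest attenuation.
# A Mixer layers a fixed chirp with one of four equally weighted random pop variants.
# Each layer passes through its own Modulator at default settings for slight variation, and the cue's overall volume is boosted.
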SoundCue /Game/Sounds/Quests/SFX/Device/Device_Call_End_Cue
root=Mixer_0
nodes [
 {id=Mixer_0, children [Modulator_1,Modulator_0], properties {inputVolume [1,1]}}
 {id=Modulator_0, children [WavePlayer_0]}
 {id=Modulator_1, children [Random_1]}
 {
  id=Random_1
  children [WavePlayer_4,WavePlayer_5,WavePlayer_6,WavePlayer_7]
  properties {bShouldExcludeFromBranchCulling=true, weights [1,1,1,1]}
 }
 {id=WavePlayer_0, properties {soundWaveAssetPtr=&/Device_Call_End_Chirp_01}}
 {id=WavePlayer_4, properties {soundWaveAssetPtr=&/Device_Call_End_Pop_01}}
 {id=WavePlayer_5, properties {soundWaveAssetPtr=&/Device_Call_End_Pop_02}}
 {id=WavePlayer_6, properties {soundWaveAssetPtr=&/Device_Call_End_Pop_03}}
 {id=WavePlayer_7, properties {soundWaveAssetPtr=&/Device_Call_End_Pop_04}}
]
properties {
 attenuationSettings=&/Fort_Narrative_Quest_Att
 maxDistance=2097152
 soundClassObject=&Fort_UI_InGame_NonCritical
 volumeMultiplier=2
}
