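# Sound cue for the incoming-call ringtone on a quest device.
# It plays a single generic incoming-call tone wave through one wave player, with no randomization or modulation.
# It uses the narrative quest attenuation, routes to the non-critical in-game UI sound class, and boosts volume well above default.
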SoundCue /Game/Sounds/Quests/SFX/Device/Device_Call_Incoming_Ringtone_Cue
root=WavePlayer_1
nodes [{id=WavePlayer_1, properties {soundWaveAssetPtr=&/Generic_Device_Call_Incoming_Tone_01}}]
properties {
 attenuationSettings=&/Fort_Narrative_Quest_Att
 maxDistance=2097152
 soundClassObject=&Fort_UI_InGame_NonCritical
 volumeMultiplier=3.5
}
